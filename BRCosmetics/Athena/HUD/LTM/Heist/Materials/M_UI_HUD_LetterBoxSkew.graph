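# UI material for a skewed letterbox shape in the Heist LTM HUD, using translucent blending.
# It rotates UVs around the (0,1) corner by a Rotation parameter and takes the green channel as a linear gradient.
# A SmoothStep between Edge and Edge plus EdgeSmooth turns that gradient into an angled mask with adjustable softness, alongside a Color parameter that defaults to white.
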Material /BRCosmetics/Athena/HUD/LTM/Heist/Materials/M_UI_HUD_LetterBoxSkew
graph {
 properties {blendMode=BLEND_Translucent, materialDomain=MD_UI, pixelDepthOffsetMode=PDOM_Legacy, usedShadingModels=Default Lit}
 nodes [
  n1 = TextureCoordinate()
  n2 = Constant2Vector() {g=1}
  n3 = ScalarParameter() {parameterName=Rotation}
  n4 = ScalarParameter() {parameterName=Edge}
  MaterialFunctionCall(
   Alpha=ComponentMask(None=MaterialFunctionCall(UVs=@n1, Rotation Center=@n2, "Rotation Angle (0-1)"=@n3) {materialFunction=&CustomRotator}.Rotated Values) {g=true}
   Min=@n4
   Max=Add(A=@n4, B=ScalarParameter() {parameterName=EdgeSmooth})
  ) {materialFunction=&SmoothStep}
  VectorParameter() {parameterName=Color, defaultValue=<rgba 1 1 1 1>}
 ]
}
